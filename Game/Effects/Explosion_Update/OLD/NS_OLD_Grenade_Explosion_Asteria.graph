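# Older Niagara grenade explosion effect (Asteria), scaled by a user Global Scale and partly tinted by Global Color.
# Emitters include flash, glow sphere, main mesh smoke burst, GPU sparks and ground smoke, all disabled in the visible portion.
# Main smoke drives dissolve-style materials through curve-fed dynamic parameters, with time-of-day scalars adjusting brightness.
# A disabled camera-facing Grid2D GPU fluid smoke/fire simulation is also present, with lighting grids and significance culling.
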NiagaraSystem /Game/Effects/Explosion_Update/OLD/NS_OLD_Grenade_Explosion_Asteria
dependencies=null
emitters {
 Flash {
  bEnabled=false
  emitterSpawnScript [/SetVariables(
    Emitter.CameraOffset:NiagaraFloat=&Multiply_Float(A=50, B=$User.Global Scale)
    Emitter.Color:LinearColor=<rgba 8 2.3613 0 0.8>
    Emitter.SpriteSize:Vector2f=&Multiply_Vector2D_ByFloat(Float=$User.Global Scale, Vector2D=<1100 20>)
   ) {isSetParameters=true}]
  emitterUpdateScript [
   &EmitterState(Life Cycle Mode=::NewEnumerator1, Loop Behavior=::NewEnumerator1, Loop Duration=0.29, UseLoopDelay=false)
   &SpawnBurst_Instantaneous(Spawn Count=1, Use Spawn Probability=false)
   /SetVariables(Emitter.SpriteSize:Vector2f=&Multiply_Vector2D_ByFloat(
     Float=&FloatFromCurve(
      FloatCurve=DataInterface() {propertyValues {Curve {keys [
          {arriveTangent=9.901, interpMode=RCIM_Cubic, leaveTangent=9.901, tangentMode=RCTM_User, time=0.45, value=0}
          {arriveTangent=12.2238, interpMode=RCIM_Cubic, leaveTangent=12.2238, tangentMode=RCTM_User, time=0.7, value=3}
          <at 0.9 0>
         ]}}}
      CurveIndex=$Emitter.NormalizedLoopAge
     )
     Vector2D=$Emitter.Initial.SpriteSize
    )) {isSetParameters=true}
  ]
  particleSpawnScript [
   /Niagara/Modules/Spawn/Initialization/InitializeParticle(
    Color=<rgba 0.2479 1 0.2583 0.2>
    Lifetime=0.25
    Mesh Scale=&VectorFromFloat(Value=&Lerp_Float(A=1, B=0.9, Alpha=&MakeFloatFromInt(Int=&SpawnGroup())))
    Module.Write Color=true
    Module.Write Lifetime=true
    Module.Write Mass=false
    Module.Write Material Random=true
    Module.Write Position=false
    Module.Write Scale=true
    Module.Write SpriteRotation=false
    Module.Write SpriteSize=false
    Position=&SimulationPosition()
    Sprite Size=<5 5>
   ) {enabled=false}
   /Niagara/Modules/Spawn/Initialization/V2/InitializeParticle(
    Color=<rgba 0.2479 1 0.2583 0.2>
    Lifetime=0.25
    Mesh Scale=&Multiply_VectorByFloat(
     Vector=&VectorFromFloat(Value=&Lerp_Float(A=1, B=0.9, Alpha=&MakeFloatFromInt(Int=&SpawnGroup())))
     Float=$User.Global Scale
    )
    Mesh Scale Mode=::NewEnumerator4
    Sprite Size Mode=::NewEnumerator4
    UsePositionOffset=false
    Write Material Random=true
   ) {module=InitializeParticle001}
   &CameraOffset(Camera Offset Amount=&Multiply_Float(A=15, B=$User.Global Scale))
  ]
  particleUpdateScript [
   &ParticleState()
   &ScaleColor(
    Scale Alpha=&FloatFromCurve(FloatCurve=DataInterface() {propertyValues {Curve {keys [{arriveTangent=-3.5509, interpMode=RCIM_Cubic, leaveTangent=-3.5509, tangentMode=RCTM_User, time=0, value=1}, <at 1 0>]}}})
    Scale RGB=&VectorFromCurve(
     VectorCurve=DataInterface() {propertyValues {XCurve {keys [<at 0 1>,<at 1 0.1>]}, YCurve {keys [<at 0 1>,<at 1 0.1>]}, ZCurve {keys [<at 0 1>,<at 1 0.1>]}}}
     Scale Curve=&Multiply_VectorByFloat(Vector=<5 3 2>, Float=&FNTimeOfDayScalar(Night Scale=2, Morning Scale=1, Evening Scale=1))
    )
    ScaleA=true
    ScaleRGB=true
   )
   &ScaleMeshSize(Scale Factor=&VectorFromFloat(Value=&FloatFromCurve(
      FloatCurve=DataInterface() {propertyValues {Curve {keys [
          {arriveTangent=4.1785, interpMode=RCIM_Cubic, leaveTangent=4.1785, tangentMode=RCTM_User, time=0.0247, value=0}
          {arriveTangent=1.7203, interpMode=RCIM_Cubic, leaveTangent=1.7203, tangentMode=RCTM_User, time=0.1, value=0.7}
          <at 0.35 0.95>
          {arriveTangent=-0.0405, interpMode=RCIM_Cubic, leaveTangent=-0.0405, tangentMode=RCTM_User, time=0.8, value=0}
         ]}}}
      Scale Curve=3.8
     )))
   /SetVariables(Particles.Color2=&ColorFromCurve(ColorCurve=DataInterface() {propertyValues {
       AlphaCurve {keys [<at 0 0.16 RCIM_Cubic>,<at 1 0 RCIM_Cubic>]}
       BlueCurve {keys [<at 0 0.5023 RCIM_Cubic>,<at 1 1 RCIM_Cubic>]}
       GreenCurve {keys [<at 0 5.4862 RCIM_Cubic>,<at 1 1 RCIM_Cubic>]}
       RedCurve {keys [<at 0 12 RCIM_Cubic>,<at 1 1 RCIM_Cubic>]}
      }})) {isSetParameters=true}
  ]
  rendererClasses [&NiagaraSpriteRendererProperties, &NiagaraMeshRendererProperties]
  renderers [
   {rendererClass=&NiagaraSpriteRendererProperties}
   {rendererClass=&NiagaraMeshRendererProperties, rendererIndex=1}
  ]
 }
 GlowSphereMesh {
  bEnabled=false
  emitterUpdateScript [
   &EmitterScalability(
    Module.bApplySpawnCountScale=false
    Module.bApplySpawnCountScaleByDistanceFraction=false
    Module.bUseMaxDistance=false
    Module.bUseMinDistance=false
   ) {enabled=false}
   &EmitterState(Life Cycle Mode=::NewEnumerator1, Loop Behavior=::NewEnumerator1, Loop Delay=0.0267, Loop Duration=0.04, UseLoopDelay=true)
   &SpawnBurst_Instantaneous(Spawn Count=1, Use Spawn Probability=false)
  ]
  particleSpawnScript [
   /Niagara/Modules/Spawn/Initialization/InitializeParticle(
    Color=&MultiplyLinearColor(Linear Color=$User.Global Color, Scale Factor=<3 3 3 1>)
    Lifetime=0.22
    Mesh Scale=<2 2 2>
    Module.Write Color=true
    Module.Write Lifetime=true
    Module.Write Mass=false
    Module.Write Material Random=true
    Module.Write Position=false
    Module.Write Scale=true
    Module.Write SpriteRotation=true
    Module.Write SpriteSize=true
    Position=&SimulationPosition()
    Sprite Rotation=&RandomRangeFloat(Minimum=0, Maximum=360)
    Sprite Size=<500 500>
   )
   /SetVariables(Particles.Scale2:Vector3f=&Multiply_VectorByFloat(Vector=&Multiply_VectorByFloat(Vector=<10 10 10>, Float=$User.Global Scale), Float=1)) {isSetParameters=true}
   &CameraOffset(Camera Offset Amount=&Multiply_Float(A=80, B=$User.Global Scale))
   /SetVariables(
    Particles.Scale=&Multiply_VectorByFloat(Vector=$Particles.Scale, Float=$User.Global Scale)
    Particles.SpriteSize=&Multiply_Vector2D_ByFloat(Float=$User.Global Scale, Vector2D=$Particles.SpriteSize)
   ) {isSetParameters=true}
  ]
  particleUpdateScript [
   &ParticleState()
   &SpriteRotationRate(
    Delta Time:NiagaraFloat=&FloatFromCurve(FloatCurve=DataInterface() {propertyValues {Curve {keys [
         {tangentMode=RCTM_User, time=0, value=1}
         {arriveTangent=-0.3174, interpMode=RCIM_Cubic, leaveTangent=-0.3174, tangentMode=RCTM_User, time=0.3, value=1}
         {arriveTangent=0.0123, interpMode=RCIM_Cubic, leaveTangent=0.0123, tangentMode=RCTM_User, time=1, value=0}
        ]}}})
    Rotation Rate=-5
   )
   &ScaleSpriteSize(
    Uniform Curve Scale=2
    Uniform Curve Sprite Scale=DataInterface() {propertyValues {Curve {keys [
        {arriveTangent=9.0725, interpMode=RCIM_Cubic, leaveTangent=9.0725, tangentMode=RCTM_User, time=0, value=0}
        {arriveTangent=-2.5249, interpMode=RCIM_Cubic, leaveTangent=-2.5249, tangentMode=RCTM_User, time=0.1408, value=0.9847}
        {arriveTangent=-0.0405, interpMode=RCIM_Cubic, leaveTangent=-0.0405, tangentMode=RCTM_User, time=1, value=0}
       ]}}}
   )
   &ScaleMeshSize(Scale Factor=&VectorFromCurve(
     VectorCurve=DataInterface() {propertyValues {
       XCurve {keys [
         {arriveTangent=4.1785, interpMode=RCIM_Cubic, leaveTangent=4.1785, tangentMode=RCTM_User, time=0, value=0}
         {arriveTangent=0.9995, interpMode=RCIM_Cubic, leaveTangent=0.9995, tangentMode=RCTM_User, time=0.2223, value=0.6825}
         {arriveTangent=-0.0405, interpMode=RCIM_Cubic, leaveTangent=-0.0405, tangentMode=RCTM_User, time=1, value=1.575}
        ]}
       YCurve {keys [
         {arriveTangent=4.1785, interpMode=RCIM_Cubic, leaveTangent=4.1785, tangentMode=RCTM_User, time=0, value=0}
         {arriveTangent=0.9995, interpMode=RCIM_Cubic, leaveTangent=0.9995, tangentMode=RCTM_User, time=0.2223, value=0.6825}
         {arriveTangent=-0.0405, interpMode=RCIM_Cubic, leaveTangent=-0.0405, tangentMode=RCTM_User, time=1, value=1.575}
        ]}
       ZCurve {keys [
         {arriveTangent=4.1785, interpMode=RCIM_Cubic, leaveTangent=4.1785, tangentMode=RCTM_User, time=0, value=0}
         {arriveTangent=0.9995, interpMode=RCIM_Cubic, leaveTangent=0.9995, tangentMode=RCTM_User, time=0.2223, value=0.6825}
         {arriveTangent=-0.0405, interpMode=RCIM_Cubic, leaveTangent=-0.0405, tangentMode=RCTM_User, time=1, value=1.575}
        ]}
      }}
     Scale Curve=<0.6 0.6 0.5>
    ))
   &ScaleColor(
    Scale Alpha=&FloatFromCurve(FloatCurve=DataInterface() {propertyValues {Curve {keys [
         {interpMode=RCIM_Cubic, tangentMode=RCTM_User, time=-0.0046, value=0.1}
         {arriveTangent=-1.0036, interpMode=RCIM_Cubic, leaveTangent=-1.0036, tangentMode=RCTM_User, time=0.6, value=0.9}
         {arriveTangent=-2.4196, interpMode=RCIM_Cubic, leaveTangent=-2.4196, tangentMode=RCTM_User, time=1.0011, value=0.002}
        ]}}})
    ScaleA=true
    ScaleRGB=false
   )
   /SetVariables(Particles.Color2=&MakeLinearColorFromVectorAndFloat(
     "Vector (RGB)"=&Lerp_Vector(
      A=<25 5 0>
      B=<16 5 1>
      Alpha=&FloatFromCurve(FloatCurve=DataInterface() {propertyValues {Curve {keys [
           {arriveTangent=-1.9247, interpMode=RCIM_Cubic, time=0, value=1}
           {arriveTangent=-1.5152, interpMode=RCIM_Cubic, leaveTangent=-1.5152, time=0.33, value=0.25}
           {interpMode=RCIM_Cubic, tangentMode=RCTM_User, time=0.66, value=0}
          ]}}})
     )
     "Float (Alpha)"=&FloatFromCurve(
      FloatCurve=DataInterface() {propertyValues {Curve {keys [{arriveTangent=-5.1417, interpMode=RCIM_Cubic, time=0.66, value=1}, <at 1.0011 0 RCIM_Cubic>]}}}
      Scale Curve=0.02
     )
    )) {isSetParameters=true}
  ]
  rendererClasses [&NiagaraMeshRendererProperties, &NiagaraSpriteRendererProperties]
  renderers [
   {rendererClass=&NiagaraMeshRendererProperties, rendererIndex=0}
   {rendererClass=&NiagaraSpriteRendererProperties, rendererIndex=1}
  ]
 }
 Grid2D_Gas_SmokeFire_Emitter {
  bEnabled=false
  simTarget=GPUComputeSim
  emitterSpawnScript [
   /NiagaraFluids/Functions/General/Fluids_Gas_Manager(
    Density Bouyancy=-0.1
    Density Color Curve=DataInterface() {propertyValues {
      ExposedName=Curve Density
      bExposeCurve=true
      AlphaCurve {keys [<at 0 0 RCIM_Cubic>,<at 2.7618 1>]}
      BlueCurve {keys [<at 0.0245 0.045>,<at 4.4467 0.055>]}
      GreenCurve {keys [<at 0.0245 0.0307>,<at 4.4467 0.0375>]}
      RedCurve {keys [<at 0.0245 0.0266>,<at 4.4467 0.0325>]}
     }}
    Depth Fade=25
    Dissipation Temperature=5
    Dissipation Velocity=0.5
    EnvironmentColor:LinearColor=<rgba 0.1427 0.2031 0.3542 1>
    Pressure Relaxation=1
    Radius Scale=0.8
    Shadowing=0.25
    SunColor:LinearColor=<rgba 1.2417 1.53 2 1>
    Temperature Bouyancy=0.0025
    Temperature Color Curve=DataInterface() {propertyValues {
      ExposedName=Curve Temperature
      bExposeCurve=true
      AlphaCurve {keys [<at 0 0>,<at 0.4853 1>,<at 0.8256 1>]}
      BlueCurve {keys [<at 0.1597 0.0069>,<at 0.7312 0.0517>,<at 0.909 4.1667>]}
      GreenCurve {keys [<at 0.1597 0.0069>,<at 0.7312 0.729>,<at 0.909 12.4907>]}
      RedCurve {keys [<at 0.1597 0.0069>,<at 0.7312 8>,<at 0.909 50>]}
     }}
    Temperature Maximum=1280
    Temperature Scale=0.2
    Vorticity Confinement=0.1
   )
   /SetVariables(Emitter.SimGrid=DataInterface() {propertyValues {
      ClearBeforeNonIterationStage=true
      NumAttributes=0
      NumCellsMaxAxis=3
      NumCellsX=1
      NumCellsY=1
      OverrideBufferFormat=Float
      PreviewAttribute=None
      SetGridFromMaxAxis=false
      bOverrideFormat=false
      bPreviewGrid=false
      RenderTargetUserParameter {parameter {name=None, typeDefHandle {registeredTypeIndex=1}, varData []}}
      WorldBBoxSize=<1200 1200>
     }}) {isSetParameters=true}
   /SetVariables(Emitter.LightingGrid=DataInterface() {propertyValues {
      ClearBeforeNonIterationStage=true
      NumAttributes=3
      NumCellsMaxAxis=3
      NumCellsX=512
      NumCellsY=512
      OverrideBufferFormat=Float
      PreviewAttribute=None
      SetGridFromMaxAxis=false
      bOverrideFormat=false
      bPreviewGrid=false
      RenderTargetUserParameter {parameter {name=None, typeDefHandle {registeredTypeIndex=4}, varData []}}
      WorldBBoxSize=<100 100>
     }}) {isSetParameters=true}
   /SetVariables(Emitter.LightingGridDownsample=DataInterface() {propertyValues {
      ClearBeforeNonIterationStage=true
      NumAttributes=1
      NumCellsMaxAxis=3
      NumCellsX=64
      NumCellsY=64
      OverrideBufferFormat=Float
      PreviewAttribute=None
      SetGridFromMaxAxis=false
      bOverrideFormat=false
      bPreviewGrid=false
      RenderTargetUserParameter {parameter {name=None, typeDefHandle {registeredTypeIndex=4}, varData []}}
      WorldBBoxSize=<100 100>
     }}) {isSetParameters=true}
   /SetVariables() {isSetParameters=true}
   /SetVariables() {isSetParameters=true}
   /SetVariables() {isSetParameters=true}
   /NiagaraFluids/Modules/Grid2D/Grid2D_SetResolution(
    Grid:NiagaraDataInterfaceGrid2DCollection=$Emitter.SimGrid
    NumCellsMaxAxis=300
    Resolution Method:ENiagaraGrid2DResolution=::NewEnumerator1
    WorldGridExtents:Vector2f=&Multiply_Vector2D_ByFloat(Float=$User.Global Scale, Vector2D=<1280 1280>)
   )
   &Grid2D_SetResolution(
    Grid:NiagaraDataInterfaceGrid2DCollection=$Emitter.LightingGridDownsample
    OtherGrid:NiagaraDataInterfaceGrid2DCollection=$Emitter.SimGrid
    Resolution Method:ENiagaraGrid2DResolution=::NewEnumerator3
    Resolution Mult=0.2
    WorldGridExtents:Vector2f=$Emitter.Grid2D_SetResolution.WorldGridExtents
   ) {module=Grid2D_SetResolution003}
   &Grid2D_SetResolution(
    Grid:NiagaraDataInterfaceGrid2DCollection=$Emitter.LightingGrid
    OtherGrid:NiagaraDataInterfaceGrid2DCollection=$Emitter.SimGrid
    Resolution Method:ENiagaraGrid2DResolution=::NewEnumerator3
    WorldGridExtents:Vector2f=$Emitter.Grid2D_SetResolution.WorldGridExtents
   ) {module=Grid2D_SetResolution004}
   &Grid2D_SetResolution(
    Grid:NiagaraDataInterfaceGrid2DCollection=$Emitter.SourceGrid
    OtherGrid:NiagaraDataInterfaceGrid2DCollection=$Emitter.SimGrid
    Resolution Method:ENiagaraGrid2DResolution=::NewEnumerator3
    WorldGridExtents:Vector2f=$Emitter.Grid2D_SetResolution.WorldGridExtents
   ) {module=Grid2D_SetResolution005}
   /SetVariables(Emitter.g:Vector3f=<0 0 -980>) {isSetParameters=true}
   &Grid2D_SetResolution(
    Grid:NiagaraDataInterfaceGrid2DCollection=$Emitter.PressureGrid
    OtherGrid:NiagaraDataInterfaceGrid2DCollection=$Emitter.SimGrid
    Resolution Method:ENiagaraGrid2DResolution=::NewEnumerator3
    WorldGridExtents:Vector2f=$Emitter.Grid2D_SetResolution.WorldGridExtents
   ) {module=Grid2D_SetResolution001}
   &Grid2D_SetResolution(
    Grid:NiagaraDataInterfaceGrid2DCollection=$Emitter.TemporaryGrid
    OtherGrid:NiagaraDataInterfaceGrid2DCollection=$Emitter.SimGrid
    Resolution Method:ENiagaraGrid2DResolution=::NewEnumerator3
    WorldGridExtents:Vector2f=$Emitter.Grid2D_SetResolution.WorldGridExtents
   ) {module=Grid2D_SetResolution002}
   /SetVariables() {isSetParameters=true}
   /SetVariables(Emitter.dt:NiagaraFloat=$Engine.DeltaTime) {enabled=false, isSetParameters=true}
   /SetVariables(Emitter.ParticleSource=DataInterface() {propertyValues {EmitterBinding {bindingMode=Other, emitterName=SourceParticles}}}) {isSetParameters=true}
   /SetVariables() {isSetParameters=true}
   &self:Grid2D_Gas_SmokeFire_Emitter.NiagaraScratchPadContainer_2.SetLightingRT_0(LightingGrid:NiagaraDataInterfaceGrid2DCollection=$Emitter.LightingGrid, LightingRT:NiagaraDataInterfaceRenderTarget2D=$Emitter.LightingRT) {module=SetLightingRT_29}
   /SetVariables() {enabled=false, isSetParameters=true}
   /SetVariables(Emitter.density=1) {isSetParameters=true}
   /NiagaraFluids/Modules/Grid2D/Grid2D_UpdateSimAttributes(Grid:NiagaraDataInterfaceGrid2DCollection=$Emitter.SimGrid, WorldGridSize:Vector2f=$Emitter.Grid2D_SetResolution.WorldGridExtents) {module=UpdateSimAttributes}
   /SetVariables() {isSetParameters=true}
   &self:Grid2D_Gas_SmokeFire_Emitter.NiagaraScratchPadContainer_2.ScratchModule_02_1(New Grid2D Collection:NiagaraDataInterfaceGrid2DCollection=$Emitter.SimGrid) {module=ScratchModule_02001}
   /SetVariables(Emitter.PressureIterations=5) {isSetParameters=true}
   /NiagaraFluids/Modules/Grid2D/Grid2D_InitRT(Grid:NiagaraDataInterfaceGrid2DCollection=$Emitter.SimGrid, SimRT:NiagaraDataInterfaceRenderTarget2D=$Emitter.SimRT) {module=Grid2D_Gas_InitRT}
   &self:Grid2D_Gas_SmokeFire_Emitter.NiagaraScratchPadContainer_2.ScratchModule_0(
    Grid:NiagaraDataInterfaceGrid2DCollection=$Emitter.SimGrid
    PressureGrid:NiagaraDataInterfaceGrid2DCollection=$Emitter.PressureGrid
    TemporaryGrid:NiagaraDataInterfaceGrid2DCollection=$Emitter.TemporaryGrid
   ) {module=ScratchModule}
  ]
  emitterUpdateScript [
   /NiagaraFN/Modules/Scalability/SignificantInstancesCull(
    Cull Response:ENiagaraExecutionStateManagement=::KillImmediately
    First Frame Only=false
    Platform Set=DataInterface() {propertyValues {Platforms {qualityLevelMask=-17, cVarConditions [], deviceProfileStates []}}}
   )
   &EmitterState(Life Cycle Mode=::NewEnumerator1, Loop Behavior=::NewEnumerator1, Loop Duration=3.6, Scalability Mode=::NewEnumerator1, UseLoopDelay=false)
   /SetVariables(Emitter.dt:NiagaraFloat=$Engine.DeltaTime) {isSetParameters=true}
   /NiagaraFluids/Modules/Grid2D/Grid2D_CreateUnitToWorldTransform(
    Camera:NiagaraDataInterfaceCamera=$Emitter.CameraQuery
    Camera Facing=true
    Offset:NiagaraPosition=$Engine.Owner.Position
    Rotation:Quat4f=$Engine.Owner.Rotation
    WorldGridExtents:Vector2f=$Emitter.Grid2D_SetResolution.WorldGridExtents
   )
   /SetVariables(Emitter.Position:NiagaraPosition=&ConvertVectorToPosition(
     Input Position=$Emitter.Grid2D_CreateUnitToWorldTransform.LocalToWorldTranslation
     Conversion Operation:ENiagara_LWCConvertVecToPos=::NewEnumerator0
    )) {isSetParameters=true}
   /NiagaraFluids/Modules/Grid2D/Grid2D_SetBounds(
    CameraFacing:NiagaraBool=$Emitter.Grid2D_CreateUnitToWorldTransform.CameraFacing
    EmitterProperties=DataInterface() {propertyValues {EmitterBinding {bindingMode=Self, emitterName=None}}}
    LocalPivot:Vector2f=$Emitter.Grid2D_CreateUnitToWorldTransform.LocalPivot
    Rotate for Camera Facing=true
    WorldGridExtents:Vector2f=$Emitter.Grid2D_SetResolution.WorldGridExtents
   )
  ]
  rendererClasses [&NiagaraSpriteRendererProperties]
  renderers [
   {rendererClass=&NiagaraSpriteRendererProperties}
   {rendererClass=&NiagaraSpriteRendererProperties, rendererIndex=1}
  ]
 }
 NE_Explosion_Main001 {
  bEnabled=false
  emitterSpawnScript [
   /SetVariables(Emitter.Depth Fade=80, Emitter.Global Color:LinearColor=$User.Global Color, Emitter.Global Scale:NiagaraFloat=$User.Global Scale) {isSetParameters=true}
   /SetVariables(
    Emitter.Color2:LinearColor=&MultiplyLinearColor(Linear Color=$Emitter.Global Color, Scale Factor=<2 1 1 1>)
    Emitter.Depth Fade:NiagaraFloat=&Multiply_Float(A=125, B=$Emitter.Global Scale)
    Emitter.Dissolve Noise Power=1.1
    Emitter.Dissolve Noise Scale:LinearColor=&MultiplyLinearColor(
     Linear Color=/Niagara/DynamicInputs/TypeConversions/MakeLinearColorFromFloat(Value:NiagaraFloat=&FloatFromCurve(FloatCurve=DataInterface() {propertyValues {Curve {keys [<at 0 2>,<at 1 1>,<at 2 0.5>]}}}, CurveIndex=$Emitter.Global Scale))
     Scale Factor=<0.0008 0.0008 0.0009 1>
    )
    Emitter.Near Camera Fade:NiagaraFloat=&Multiply_Float(A=400, B=$Emitter.Global Scale)
   ) {isSetParameters=true}
   /SetVariables(Emitter.Dissolve Scale Adjsutment:NiagaraFloat=&FloatFromCurve(
     FloatCurve=DataInterface() {propertyValues {Curve {keys [<at 0 3>,<at 1 1>,<at 5 0.5>,<at 20 0.4>,<at 100 0.4>]}}}
     CurveIndex=$Emitter.Global Scale
     Scale Curve=$Emitter.Global Scale
    )) {isSetParameters=true}
  ]
  emitterUpdateScript [
   &EmitterState(Life Cycle Mode=::NewEnumerator1, Loop Behavior=::NewEnumerator1, Loop Delay=0.1383, Loop Duration=0.627, UseLoopDelay=true)
   &SpawnBurst_Instantaneous(Spawn Group=1, Use Spawn Probability=false, UseLoopCountLimit=false) {module=SpawnBurst_Instantaneous002}
   &SpawnBurst_Instantaneous(Spawn Count=5, Spawn Time=0.01, Use Spawn Probability=false, UseLoopCountLimit=false) {module=SpawnBurst_Instantaneous001}
  ]
  particleSpawnScript [
   /Niagara/Modules/Spawn/Initialization/V2/InitializeParticle(
    Color=&MaskLinearColorBySpawnGroup(Masked Linear Color=<rgba 0.3229 0.3028 0.2791 1>, Passthrough Linear Color=<rgba 0.2969 0.2679 0.2401 1>, Spawn Group Mask=1)
    Lifetime=&MaskFloatBySpawnGroup(
     Masked Float=2.1
     Passthrough Float=&SelectFloatFromArray(
      Array Sampling Mode=::NewEnumerator0
      Direct Array Index=?
      Float Selection Array=DataInterface() {propertyValues {FloatData [1.5,1.95]}}
      Random Seed=0
     )
     Spawn Group Mask=1
    )
    Mesh Scale=&Multiply_VectorByFloat(
     Vector=&MaskVectorBySpawnGroup(
      Masked Vector=<3 3.5 2.6>
      Passthrough Vector=&RandomRangeVector(Minimum=<3 2.8 2.8>, Maximum=<3 3.5 3.5>)
      Spawn Group Mask=1
     )
     Float=&MaskFloatBySpawnGroup(
      Masked Float=1
      Passthrough Float=&FloatFromCurve(
       FloatCurve=DataInterface() {propertyValues {Curve {keys [<at 0 0.9 RCIM_Constant>,<at 0.5 1 RCIM_Constant>,<at 1 1>]}}}
       CurveIndex=&RandomRangeFloat(Minimum=0, Maximum=1)
       Scale Curve=0.7
      )
      Spawn Group Mask=1
     )
    )
    Mesh Scale Mode=::NewEnumerator4
    UsePositionOffset=false
    Write Material Random=true
   )
   &ShapeLocation(
    MaskSpawnGroup=false
    Override Seed=false
    Sphere Radius=&Multiply_Float(A=40, B=$Emitter.Global Scale)
    UseOwnerScale=true
   )
   &AddVelocity(
    Module.Use Falloff Distance=false
    Origin Offset=&Multiply_VectorByFloat(Vector=&RandomRangeVector(Minimum=<0 0 -10>, Maximum=<0 0 -10>), Float=$Emitter.Global Scale)
    Override Seed=false
    Velocity Mode=::NewEnumerator1
    Velocity Speed=&MaskFloatBySpawnGroup(
     Masked Float=0
     Passthrough Float=&Multiply_Float(A=&RandomRangeFloat(Minimum=900, Maximum=1100), B=$Emitter.Global Scale)
     Spawn Group Mask=1
    )
   )
   &CameraOffset(Camera Offset Amount=&Multiply_Float(A=&MaskFloatBySpawnGroup(Masked Float=70, Passthrough Float=40, Spawn Group Mask=1), B=$Emitter.Global Scale))
   /SetVariables(Particles.Scale=&Multiply_VectorByFloat(Vector=$Particles.Scale, Float=$Emitter.Global Scale)) {isSetParameters=true}
  ]
  particleUpdateScript [
   &ParticleState()
   &GravityForce(Gravity=&Multiply_VectorByFloat(
     Vector=&MaskVectorBySpawnGroup(
      Masked Vector=<0 0 1600>
      Passthrough Vector=&RandomRangeVector(Minimum=<0 0 500>, Maximum=<1 1 1500>)
      Spawn Group Mask=1
     )
     Float=$Emitter.Global Scale
    ))
   &Drag(Drag=&RandomRangeFloat(Minimum=9, Maximum=10), Use Linear Drag=true, Use Rotational Drag=false)
   &DynamicMaterialParameters(
    Index 0 Param 1=&FloatFromCurve(
     FloatCurve=DataInterface() {propertyValues {Curve {keys [{arriveTangent=-2.2221, interpMode=RCIM_Cubic, leaveTangent=-2.2221, tangentMode=RCTM_User, time=0.1, value=1}, <at 0.95 0>]}}}
     Scale Curve=&Multiply_Float(A=-130, B=$Emitter.Global Scale)
    )
    Index 0 Param 2=&FloatFromCurve(
     FloatCurve=DataInterface() {propertyValues {Curve {keys [
         {tangentMode=RCTM_User, time=0, value=1}
         {arriveTangent=-0.0405, interpMode=RCIM_Cubic, leaveTangent=-0.0405, tangentMode=RCTM_User, time=0.2, value=1}
         {arriveTangent=-0.0002, interpMode=RCIM_Cubic, leaveTangent=-0.0002, tangentMode=RCTM_User, time=0.9, value=0.6}
        ]}}}
     Scale Curve=0.8
    )
    Index 0 Param 3=&FloatFromCurve(
     FloatCurve=DataInterface() {propertyValues {Curve {keys [
         {arriveTangent=1.4725, interpMode=RCIM_Cubic, leaveTangent=1.4725, tangentMode=RCTM_User, time=0, value=0}
         {tangentMode=RCTM_User, time=1, value=1}
        ]}}}
     Scale Curve=&Multiply_Float(
      A=&MaskFloatBySpawnGroup(Masked Float=250, Passthrough Float=&RandomRangeFloat(Minimum=250, Maximum=300), Spawn Group Mask=1)
      B=$Emitter.Global Scale
     )
    )
    Index 0 Param 4=&MaskFloatBySpawnGroup(
     Masked Float=&FloatFromCurve(
      FloatCurve=DataInterface() {propertyValues {Curve {keys [
          {arriveTangent=0.1175, interpMode=RCIM_Cubic, leaveTangent=0.6205, tangentMode=RCTM_Break, time=0.22, value=-2}
          {arriveTangent=6.3488, interpMode=RCIM_Cubic, leaveTangent=6.3488, tangentMode=RCTM_User, time=0.5987, value=-0.8048}
          {arriveTangent=42.4007, interpMode=RCIM_Cubic, leaveTangent=14.4841, tangentMode=RCTM_Break, time=1, value=7.3}
         ]}}}
      Scale Curve=0.2
     )
     Passthrough Float=&FloatFromCurve(
      FloatCurve=DataInterface() {propertyValues {Curve {keys [
          {arriveTangent=0.1175, interpMode=RCIM_Cubic, leaveTangent=1.7315, tangentMode=RCTM_Break, time=0.1, value=-3}
          {arriveTangent=35.5342, interpMode=RCIM_Cubic, leaveTangent=14.4841, tangentMode=RCTM_Break, time=1, value=9}
         ]}}}
      Scale Curve=0.2
     )
     Spawn Group Mask=1
    )
    Index 1 Param 1=&FloatFromCurve(
     FloatCurve=DataInterface() {propertyValues {Curve {keys [
         {arriveTangent=0.5348, interpMode=RCIM_Cubic, leaveTangent=0.5348, tangentMode=RCTM_User, time=0, value=0.01}
         {arriveTangent=1.457, interpMode=RCIM_Cubic, leaveTangent=1.457, tangentMode=RCTM_User, time=0.04, value=0.05}
         {arriveTangent=85.7775, interpMode=RCIM_Cubic, leaveTangent=85.7775, tangentMode=RCTM_User, time=0.08, value=16}
        ]}}}
     Scale Curve=&MaskFloatBySpawnGroup(Masked Float=20, Passthrough Float=&RandomRangeFloat(Minimum=20, Maximum=20), Spawn Group Mask=1)
    )
    Index 1 Param 2=&Divide_Float(
     A=&MaskFloatBySpawnGroup(
      Masked Float=0.014
      Passthrough Float=&RandomRangeFloat(Minimum=0.012, Maximum=0.014)
      Spawn Group Mask=1
     )
     B=$Emitter.Global Scale
    )
    Index 1 Param 3=&FloatFromCurve(
     FloatCurve=DataInterface() {propertyValues {Curve {keys [{interpMode=RCIM_Cubic, tangentMode=RCTM_User, time=0.4, value=0.5}, <at 0.8 0.1>]}}}
     Scale Curve=5
    )
    Index 1 Param 4=&FloatFromCurve(FloatCurve=DataInterface() {propertyValues {Curve {keys [{interpMode=RCIM_Cubic, tangentMode=RCTM_User, time=0.05, value=1}, {interpMode=RCIM_Cubic, tangentMode=RCTM_User, time=1, value=6}]}}})
    Index 2 Param 1=&FloatFromCurve(
     FloatCurve=DataInterface() {propertyValues {Curve {keys [
         {arriveTangent=0.9553, interpMode=RCIM_Cubic, leaveTangent=0.9553, tangentMode=RCTM_User, time=0, value=0}
         {arriveTangent=1.5886, interpMode=RCIM_Cubic, leaveTangent=1.5886, tangentMode=RCTM_User, time=1, value=1.5}
        ]}}}
     Scale Curve=1.2
    )
    Index 2 Param 2=&FloatFromCurve(
     FloatCurve=DataInterface() {propertyValues {Curve {keys [
         {interpMode=RCIM_Cubic, tangentMode=RCTM_User, time=0.3, value=0}
         {arriveTangent=2.4333, interpMode=RCIM_Cubic, leaveTangent=2.4333, tangentMode=RCTM_User, time=1, value=1}
        ]}}}
     Scale Curve=&Multiply_Float(A=0.02, B=$Emitter.Global Scale)
    )
    Index 2 Param 3=&MaskFloatBySpawnGroup(
     Masked Float=&FloatFromCurve(
      FloatCurve=DataInterface() {propertyValues {Curve {keys [
          {arriveTangent=-17.9351, interpMode=RCIM_Cubic, leaveTangent=-17.9351, tangentMode=RCTM_User, time=0.02, value=1}
          {arriveTangent=-0.7527, interpMode=RCIM_Cubic, leaveTangent=-0.7527, tangentMode=RCTM_User, time=0.1, value=0}
         ]}}}
      Scale Curve=50
     )
     Passthrough Float=&FloatFromCurve(
      FloatCurve=DataInterface() {propertyValues {Curve {keys [
          {arriveTangent=-12.2097, interpMode=RCIM_Cubic, leaveTangent=-12.2097, tangentMode=RCTM_User, time=0.01, value=1}
          {arriveTangent=-0.7527, interpMode=RCIM_Cubic, leaveTangent=-0.7527, tangentMode=RCTM_User, time=0.12, value=0}
         ]}}}
      Scale Curve=20
     )
     Spawn Group Mask=1
    )
    Index 2 Param 4=$Emitter.Global Scale
    Index 3 Param 1=&FloatFromCurve(
     FloatCurve=DataInterface() {propertyValues {Curve {keys [
         {arriveTangent=2.2353, interpMode=RCIM_Cubic, leaveTangent=2.2353, tangentMode=RCTM_User, time=0, value=0.2}
         {arriveTangent=1.4416, interpMode=RCIM_Cubic, leaveTangent=1.4416, tangentMode=RCTM_User, time=0.2, value=0.5}
         {interpMode=RCIM_Cubic, tangentMode=RCTM_User, time=0.8, value=1.25}
        ]}}}
     Scale Curve=&Multiply_Float(A=-2.5, B=$Emitter.Global Scale)
    )
    Index 3 Param 2=&FloatFromCurve(
     FloatCurve=DataInterface() {propertyValues {Curve {keys [
         {arriveTangent=5.1563, interpMode=RCIM_Cubic, leaveTangent=5.1563, tangentMode=RCTM_User, time=0, value=0}
         {arriveTangent=0.1309, interpMode=RCIM_Cubic, leaveTangent=0.1309, tangentMode=RCTM_User, time=0.1699, value=0.4}
         {arriveTangent=0.8545, interpMode=RCIM_Cubic, leaveTangent=0.8545, tangentMode=RCTM_User, time=1, value=1}
        ]}}}
     Scale Curve=&Multiply_Float(A=-0.4, B=$Emitter.Global Scale)
    )
    Index 3 Param 3=&FloatFromCurve(
     FloatCurve=DataInterface() {propertyValues {Curve {keys [
         {interpMode=RCIM_Cubic, tangentMode=RCTM_User, time=0, value=16}
         {arriveTangent=-5.9757, interpMode=RCIM_Cubic, leaveTangent=-5.9757, tangentMode=RCTM_User, time=0.1, value=2.5}
         {arriveTangent=-5.9757, interpMode=RCIM_Cubic, leaveTangent=-5.9757, tangentMode=RCTM_User, time=0.15, value=1.25}
        ]}}}
     Scale Curve=&FNTimeOfDayScalar(Night Scale=0.6, Morning Scale=0.6, Evening Scale=0.6)
    )
    Index 3 Param 4=&FloatFromCurve(
     FloatCurve=DataInterface() {propertyValues {Curve {keys [{tangentMode=RCTM_User, time=0, value=0}, {tangentMode=RCTM_User, time=1, value=1}]}}}
     Scale Curve=&Multiply_Float(A=-100, B=$Emitter.Global Scale)
    )
    Module.Index0Param0Write=true
    Module.Index0Param1Write=true
    Module.Index0Param2Write=true
    Module.Index0Param3Write=true
    Module.Index1Param0Write=true
    Module.Index1Param1Write=true
    Module.Index1Param2Write=true
    Module.Index1Param3Write=true
    Module.Index2Param0Write=true
    Module.Index2Param1Write=true
    Module.Index2Param2Write=true
    Module.Index2Param3Write=true
    Module.Index3Param0Write=true
    Module.Index3Param1Write=true
    Module.Index3Param2Write=true
    Module.Index3Param3Write=true
    Write Parameter Index 1=true
    Write Parameter Index 2=true
    Write Parameter Index 3=true
   ) {module=DynamicMaterialParameters001}
   &ScaleColor(
    Linear Color Curve=DataInterface() {propertyValues {
      AlphaCurve {keys [<at 0 1>,<at 0.6968 1>,<at 0.9894 0>]}
      BlueCurve {keys [<at 0.1888 0.6042>,<at 0.4977 0.4497>]}
      GreenCurve {keys [<at 0.1888 0.6042>,<at 0.4977 0.4497>]}
      RedCurve {keys [<at 0.1888 0.6042>,<at 0.4977 0.4497>]}
     }}
    Scale Mode=::NewEnumerator2
   )
   &CurlNoiseForce(
    Noise Frequency=&Divide_Float(A=1, B=$Emitter.Global Scale)
    Noise Quality / Cost=::NewEnumerator0
    Noise Strength=&Multiply_Float(
     A=&FloatFromCurve(
      FloatCurve=DataInterface() {propertyValues {Curve {keys [
          {arriveTangent=3.7261, interpMode=RCIM_Cubic, leaveTangent=3.7261, tangentMode=RCTM_User, time=0.05, value=0}
          {interpMode=RCIM_Cubic, tangentMode=RCTM_User, time=0.5, value=1}
          <at 0.8 0>
         ]}}}
      Scale Curve=1000
     )
     B=$Emitter.Global Scale
    )
    Pan Noise=true
    Pan Noise Field=&RandomRangeVector(Minimum=<0 0 0>, Maximum=<1 1 1>)
   )
   &SolveForcesAndVelocity(Clamp Velocity=false, Limit Acceleration=false)
   &ScaleMeshSize(Scale Factor=&VectorFromCurve(
     VectorCurve=DataInterface() {propertyValues {
       XCurve {keys [
         {arriveTangent=1.1946, interpMode=RCIM_Cubic, leaveTangent=1.1946, tangentMode=RCTM_User, time=0, value=0.7}
         {arriveTangent=0.581, interpMode=RCIM_Cubic, leaveTangent=0.581, tangentMode=RCTM_User, time=0.2, value=0.8875}
         {arriveTangent=0.0837, interpMode=RCIM_Cubic, leaveTangent=0.0837, tangentMode=RCTM_User, time=1, value=1.1}
        ]}
       YCurve {keys [
         {arriveTangent=1.1946, interpMode=RCIM_Cubic, leaveTangent=1.1946, tangentMode=RCTM_User, time=0, value=0.7}
         {arriveTangent=0.581, interpMode=RCIM_Cubic, leaveTangent=0.581, tangentMode=RCTM_User, time=0.2, value=0.8875}
         {arriveTangent=0.0837, interpMode=RCIM_Cubic, leaveTangent=0.0837, tangentMode=RCTM_User, time=1, value=1.1}
        ]}
       ZCurve {keys [
         {arriveTangent=0.9745, interpMode=RCIM_Cubic, leaveTangent=0.9745, tangentMode=RCTM_User, time=0, value=0.7}
         {arriveTangent=0.8594, interpMode=RCIM_Cubic, leaveTangent=0.8594, tangentMode=RCTM_User, time=0.2, value=0.8875}
         {arriveTangent=0.0837, interpMode=RCIM_Cubic, leaveTangent=0.0837, tangentMode=RCTM_User, time=1, value=1.1}
        ]}
      }}
     Scale Curve=&Multiply_VectorByFloat(
      Vector=<1 1 1>
      Float=&FloatFromCurve(
       FloatCurve=DataInterface() {propertyValues {Curve {keys [<at 0.5 0.8 RCIM_Constant>,<at 0.7 0.9 RCIM_Constant>]}}}
       CurveIndex=&RandomRangeFloat(Minimum=0.4, Maximum=1)
       Scale Curve=&MaskFloatBySpawnGroup(
        Masked Float=&FloatFromCurve(
         FloatCurve=DataInterface() {propertyValues {Curve {keys [
             {arriveTangent=0.8977, interpMode=RCIM_Cubic, leaveTangent=0.8977, tangentMode=RCTM_User, time=0, value=0.6}
             {arriveTangent=0.4659, interpMode=RCIM_Cubic, leaveTangent=0.4659, tangentMode=RCTM_User, time=0.7, value=1}
            ]}}}
         Scale Curve=0.88
        )
        Passthrough Float=1
        Spawn Group Mask=1
       )
      )
     )
    ))
  ]
  rendererClasses [&NiagaraMeshRendererProperties]
  renderers [{rendererClass=&NiagaraMeshRendererProperties, rendererIndex=0}]
 }
 NE_Sparks_GPU {
  bEnabled=false
  simTarget=GPUComputeSim
  emitterSpawnScript [
   /SetVariables(
    Emitter.Global Color:LinearColor=$User.Global Color
    Emitter.Global Scale:NiagaraFloat=$User.Global Scale
    Emitter.RandomFloat:NiagaraFloat=&RandomRangeFloat(Minimum=-180, Maximum=180)
    Emitter.RandomInt:NiagaraInt32=&RandomRangeInt(Minimum=0, Maximum=4)
   ) {isSetParameters=true}
   /SetVariables(Emitter.RandVector:Vector3f=&RandomRangeVector(
     Minimum=&SelectVectorFromArray(
      Direct Array Index=$Emitter.RandomInt
      Vector Selection Array=DataInterface() {propertyValues {FloatData [<0.5 1 0.2>, <1 0.5 0.5>, <-1 1 0.2>, <1 -1 0.5>, <-1 -1 0.2>]}}
     )
     Maximum=&SelectVectorFromArray(
      Direct Array Index=$Emitter.RandomInt
      Vector Selection Array=DataInterface() {propertyValues {FloatData [<0.5 1 2>, <1 0.5 2>, <-1 1 2>, <1 -1 2>, <-1 -1 2>]}}
     )
    )) {isSetParameters=true}
  ]
  emitterUpdateScript [
   &EmitterState(
    Enable Distance Culling=true
    Life Cycle Mode=::NewEnumerator1
    Loop Behavior=::NewEnumerator1
    Loop Delay=0.01
    Loop Duration=0.1
    Max Distance Response=::KillAfterParticlesFinish
    MaxDistance=4000
    Module.bUseMaxDistance=true
    Module.bUseMinDistance=false
    Scalability Mode=::NewEnumerator1
    UseLoopDelay=true
   )
   &SpawnRate(SpawnRate=5000, Use Spawn Probability=false)
  ]
  particleSpawnScript [
   &PartitionParticles(Number Of Partitions=5)
   &InitializeParticle(
    Color=$Emitter.Global Color
    Lifetime=&FloatFromCurve(
     FloatCurve=DataInterface() {propertyValues {Curve {keys [<at 0 0.25 RCIM_Constant>,<at 0.3 0.45>,<at 0.7 1 RCIM_Constant>,<at 0.85 1.5>,<at 1 2>]}}}
     CurveIndex=&RandomRangeFloat(Minimum=0, Maximum=1)
     Scale Curve=0.35
    )
    Mesh Scale Mode=::NewEnumerator1
    Mesh Uniform Scale Max=1.7
    Mesh Uniform Scale Min=0.6
    Sprite Rotation Mode=::NewEnumerator2
    Sprite Size Mode=::NewEnumerator3
    Uniform Sprite Size=&Multiply_Float(A=&RandomRangeFloat(Minimum=7, Maximum=11), B=$Emitter.Global Scale)
    UsePositionOffset=false
    Write Material Random=true
   )
   &SystemLocation(
    MaskSpawnGroup=false
    Offset=&Multiply_Vector(
     A=$Emitter.RandVector
     B=&SelectVectorFromArray(
      Direct Array Index=$Particles.PartitionParticles.Partition
      Direct Array Mode=::NewEnumerator1
      Vector Selection Array=DataInterface() {propertyValues {FloatData [<0.1 0.1 0.1>, <-0.1 -0.1 0.2>, <0 0.1 0.4>, <0.1 0 0.3>, <0.1 -0.1 0.5>]}}
     )
    )
   )
   &AddVelocity(
    Cone Angle=4
    Cone Axis=&Multiply_Vector(
     A=&SelectVectorFromArray(
      Direct Array Index=$Particles.PartitionParticles.Partition
      Vector Selection Array=DataInterface() {propertyValues {FloatData [<0 0.1 1>, <-0.5 0 1>, <1 1 0.1>, <-1 1 1>, <1 -1 0.4>]}}
     )
     B=$Emitter.RandVector
    )
    Module.UseConeFalloff=false
    Module.UseVelDistribution=false
    Override Seed=false
    Rotation Mode=::NewEnumerator2
    Velocity Mode=::NewEnumerator2
    Velocity Speed=&Multiply_Float(
     A=&Multiply_Float(
      A=&RandomRangeFloat(Minimum=1800, Maximum=2200)
      B=&ReturnNormalizedExecIndex()
     )
     B=$Emitter.Global Scale
    )
    Yaw / Pitch / Roll:Vector3f=&MakeVector(X=$Emitter.RandomFloat, Y=0, Z=0)
   ) {module=AddVelocity001}
   &JitterPosition(Jitter Amount:NiagaraFloat=&Multiply_Float(A=3, B=$Emitter.Global Scale))
  ]
  particleUpdateScript [
   &ParticleState()
   &GravityForce(Gravity=&Multiply_VectorByFloat(Vector=&RandomRangeVector(Minimum=<0 0 -900>, Maximum=<0 0 -650>), Float=$Emitter.Global Scale))
   &Drag(Drag=&RandomRangeFloat(Minimum=2, Maximum=2.5), Use Linear Drag=true, Use Rotational Drag=false)
   &CurlNoiseForce(
    Noise Frequency=&Divide_Float(A=15, B=$Emitter.Global Scale)
    Noise Quality / Cost=::NewEnumerator0
    Noise Strength=&Multiply_Float(A=500, B=$Emitter.Global Scale)
    Pan Noise=false
   )
   &ScaleColor(
    Scale Alpha=&FloatFromCurve(FloatCurve=DataInterface() {propertyValues {Curve {keys [
         {arriveTangent=26.7273, interpMode=RCIM_Cubic, leaveTangent=26.7273, tangentMode=RCTM_User, time=-0.0092, value=0.0144}
         {interpMode=RCIM_Cubic, tangentMode=RCTM_User, time=0.1, value=1}
         {interpMode=RCIM_Cubic, tangentMode=RCTM_User, time=0.85, value=1}
         {arriveTangent=-1.3701, interpMode=RCIM_Cubic, leaveTangent=-1.3701, tangentMode=RCTM_User, time=1.0011, value=0}
        ]}}})
    Scale RGB=&Lerp_Vector(
     A=<0 0 0>
     B=<12.5 4 0.8>
     Alpha=&FloatFromCurve(FloatCurve=DataInterface() {propertyValues {Curve {keys [
          {arriveTangent=15.3906, interpMode=RCIM_Cubic, leaveTangent=15.3906, tangentMode=RCTM_User, time=-0.0092, value=1}
          {interpMode=RCIM_Cubic, tangentMode=RCTM_User, time=0.0967, value=0.8}
          {interpMode=RCIM_Cubic, tangentMode=RCTM_User, time=0.45, value=0.8}
          <at 0.7 0 RCIM_Cubic>
         ]}}})
    )
    ScaleA=true
    ScaleRGB=true
   )
   &SolveForcesAndVelocity(Clamp Velocity=false, Limit Acceleration=false)
   &ScaleSpriteSize(Scale Factor=&Vector2DFromFloat(Value=&FloatFromCurve(FloatCurve=DataInterface() {propertyValues {Curve {keys [<at -0.0046 0>,<at 0.025 1>,<at 0.66 1>,<at 1.0011 0.5>]}}}, Scale Curve=1.5)))
   &ScaleSpriteSizeBySpeed(Max Scale Factor=<0.5 4>, Min Scale Factor=<1 0.6>, Velocity Threshold=1250)
  ]
  rendererClasses [&NiagaraSpriteRendererProperties]
  renderers [{rendererClass=&NiagaraSpriteRendererProperties}]
 }
 Smoke_Ground001 {
  bEnabled=false
  emitterSpawnScript [/SetVariables(Emitter.Global Scale:NiagaraFloat=$User.Global Scale) {isSetParameters=true}]
  emitterUpdateScript [
   &EmitterState(Life Cycle Mode=::NewEnumerator1, Loop Behavior=::NewEnumerator1, Loop Delay=0.02, Loop Duration=1.8761, UseLoopDelay=true)
   &SpawnBurst_Instantaneous(Spawn Count=17, Spawn Time=0.05, Use Spawn Probability=false)
   /SetVariables(
    Emitter. Color:LinearColor=&ColorFromCurve(
     ColorCurve=DataInterface() {propertyValues {
       AlphaCurve {keys [<at 0 1 RCIM_Cubic>,<at 0.1176 0 RCIM_Cubic>]}
       BlueCurve {keys [
         <at 0 1.9167>
         {arriveTangent=-22.1709, interpMode=RCIM_Cubic, leaveTangent=-22.1709, time=0.0418, value=0}
         <at 0.0836 0.0625>
         <at 0.1176 0>
         <at 1 0 RCIM_Cubic>
        ]}
       GreenCurve {keys [
         <at 0 6.3255>
         {arriveTangent=-67.6362, interpMode=RCIM_Cubic, leaveTangent=-67.6362, time=0.0418, value=7.7363}
         <at 0.0836 0.669>
         <at 0.1176 0>
         <at 1 0 RCIM_Cubic>
        ]}
       RedCurve {keys [
         <at 0 16>
         {arriveTangent=-167.4025, interpMode=RCIM_Cubic, leaveTangent=-167.4025, time=0.0418, value=50}
         <at 0.0836 2>
         <at 0.1176 0>
         <at 1 0 RCIM_Cubic>
        ]}
      }}
     CurveIndex=$Emitter.NormalizedLoopAge
    )
    Emitter.Exponent=1.5
    Emitter.Position:NiagaraPosition=&Multiply_VectorByFloat(
     Vector=&AddVectorToPosition(
      Position=$Engine.Owner.Position
      Vector=&MakeVector(
       X=0
       Y=0
       Z=&Multiply_Float(
        A=$Emitter.Global Scale
        B=&FloatFromCurve(
         FloatCurve=DataInterface() {propertyValues {Curve {keys [{arriveTangent=0.3523, interpMode=RCIM_Cubic, leaveTangent=0.3523, tangentMode=RCTM_User, time=0, value=1.25}, <at 0.5 1.5>]}}}
         CurveIndex=$Engine.System.Age
         Scale Curve=300
        )
       )
      )
     )
     Float=1
    )
    Emitter.Radius:NiagaraFloat=&Multiply_Float(
     A=&FloatFromCurve(
      FloatCurve=DataInterface() {propertyValues {Curve {keys [{interpMode=RCIM_Cubic, tangentMode=RCTM_User, time=0.1, value=0.95}, <at 0.5 0>]}}}
      CurveIndex=$Emitter.NormalizedLoopAge
      Scale Curve=600
     )
     B=$Emitter.Global Scale
    )
    Emitter.VisibilityTag=1
   ) {isSetParameters=true}
  ]
  particleSpawnScript [
   /SetVariables() {enabled=false, isSetParameters=true}
   &InitializeParticle(
    Color=<rgba 0.0781 0.0781 0.0781 1.1>
    Lifetime Max=1.5
    Lifetime Min=0.5
    Lifetime Mode=::NewEnumerator1
    Sprite Rotation Mode=::NewEnumerator2
    Sprite Size Mode=::NewEnumerator1
    Sprite UV Mode=::NewEnumerator3
    Uniform Sprite Size Max=450
    Uniform Sprite Size Min=400
    UsePositionOffset=true
    Write Material Random=true
   )
   &ShapeLocation(
    MaskSpawnGroup=false
    Override Seed=false
    Radius Position=1.2
    Ring / Disc Distribution Mode:ENiagaraSphereDistributionMode=::NewEnumerator1
    Ring Radius=&Multiply_Float(A=50, B=$User.Global Scale)
    Shape Primitive=::NewEnumerator4
    U Position=&ReturnNormalizedExecIndex(Normalized Index Includes One=false)
    UseOwnerScale=true
   )
   &AddVelocityFromPoint(
    Module.Use Falloff Distance=false
    Origin Offset=&Multiply_VectorByFloat(Vector=<0 0 -5>, Float=$Emitter.Global Scale)
    Velocity Origin:Vector3f=&SimulationPosition()
    Velocity Strength=&Multiply_Float(A=&UniformRangedFloat(Minimum=2000, Maximum=2500), B=$Emitter.Global Scale)
   )
   &ApplyInitialForces()
   &CameraOffset(Camera Offset Amount=&Multiply_Float(A=40, B=$Emitter.Global Scale))
   /SetVariables(
    Particles.Stored Particle Size:Vector3f=$Particles.Scale
    Particles.Stored Sprite Size:Vector2f=&Multiply_Vector2D_ByFloat(Float=$Emitter.Global Scale, Vector2D=$Particles.SpriteSize)
   ) {isSetParameters=true}
  ]
  particleUpdateScript [
   &ParticleState()
   /SetVariables(Particles.Scale=$StackContext.Stored Particle Size, Particles.SpriteSize=$StackContext.Stored Sprite Size) {isSetParameters=true}
   &AccelerationForce(Acceleration=&MakeVector(X=0, Y=0, Z=&UniformRangedFloat(Minimum=-100, Maximum=200)))
   &Drag(Drag=&RandomRangeFloat(Minimum=7, Maximum=9), Use Linear Drag=true, Use Rotational Drag=false)
   &self:ViewpointBased_Sprite_Rotation_03(Rotation Rate=3) {module=ViewpointBased_Sprite_Rotation_03}
   &SolveForcesAndVelocity(Clamp Velocity=false, Limit Acceleration=false)
   &ScaleSpriteSize(Uniform Curve Sprite Scale=DataInterface() {propertyValues {Curve {keys [
        {arriveTangent=0.8609, interpMode=RCIM_Cubic, leaveTangent=0.8609, tangentMode=RCTM_User, time=0, value=0.15}
        {arriveTangent=2.1878, interpMode=RCIM_Cubic, leaveTangent=2.1878, tangentMode=RCTM_User, time=0.4, value=0.85}
        <at 1 1.25 RCIM_Cubic>
       ]}}})
   &DynamicMaterialParameters(
    Index 0 Param 1=&FloatFromCurve(FloatCurve=DataInterface() {propertyValues {Curve {keys [<at -0.0046 1.0113>,<at 0.2 1>,<at 1.0011 -1>]}}})
    Index 0 Param 4=0.2
    Module.Param0WriteEnabled=true
    Module.Param1WriteEnabled=false
    Module.Param2WriteEnabled=false
    Module.Param3WriteEnabled=true
   ) {module=DynamicMatParams}
   &ScaleColor(
    Scale Alpha=&FloatFromCurve(FloatCurve=DataInterface() {propertyValues {Curve {keys [
         {arriveTangent=6.7279, interpMode=RCIM_Cubic, leaveTangent=6.7279, tangentMode=RCTM_User, time=0.0025, value=0.5379}
         {arriveTangent=-0.5367, interpMode=RCIM_Cubic, leaveTangent=-0.5367, time=0.1843, value=0.9779}
         <at 1.0011 0.002 RCIM_Cubic>
        ]}}})
    Scale RGB=&MakeVectorFromLinearColorRGB(Color=&ColorFromCurve(ColorCurve=DataInterface() {propertyValues {
        AlphaCurve {keys [<at 0 1 RCIM_Cubic>,<at 1 0 RCIM_Cubic>]}
        BlueCurve {keys [<at 0 1.14 RCIM_Cubic>,<at 1 1 RCIM_Cubic>]}
        GreenCurve {keys [<at 0 1.842 RCIM_Cubic>,<at 1 1 RCIM_Cubic>]}
        RedCurve {keys [<at 0 3 RCIM_Cubic>,<at 1 1 RCIM_Cubic>]}
       }}))
    ScaleA=true
    ScaleRGB=true
   )
   /Game/Effects/Explosion_Update/ModuleScripts/Camera_Distance_Fade(
    Near FadeIn Distance:NiagaraFloat=&Multiply_Float(A=200, B=$Emitter.Global Scale)
    Near Fadeout Distance:NiagaraFloat=&Multiply_Float(A=300, B=$Emitter.Global Scale)
    Stored Particle Scale:Vector3f=$Particles.Stored Particle Size
    Stored SpriteSize:Vector2f=$Particles.Stored Sprite Size
   )
   &SubUVAnimation(SubUV Loop Count=0.5)
  ]
  rendererClasses [&NiagaraSpriteRendererProperties, &NiagaraLightRendererProperties]
  renderers [
   {rendererClass=&NiagaraSpriteRendererProperties}
   {rendererClass=&NiagaraLightRendererProperties, rendererIndex=1}
  ]
 }
 SourceParticles {
  bEnabled=false
  simTarget=GPUComputeSim
  emitterUpdateScript [
   &SignificantInstancesCull(
    Cull Response:ENiagaraExecutionStateManagement=::KillImmediately
    First Frame Only=false
    Platform Set=DataInterface() {propertyValues {Platforms {qualityLevelMask=-17, cVarConditions [], deviceProfileStates []}}}
   )
   &EmitterState(
    Life Cycle Mode=::NewEnumerator1
    Loop Behavior=::NewEnumerator1
    Loop Delay=0.015
    Loop Duration=0.2
    Scalability Mode=::NewEnumerator1
    UseLoopDelay=true
   )
   &SpawnBurst_Instantaneous(
    Spawn Count=&RandomRangeInt(Minimum=5, Maximum=7, Evaluation Type=::NewEnumerator1)
    Use Spawn Probability=false
   )
   &SpawnRate(SpawnRate=55, Use Spawn Probability=false) {enabled=false}
   &Grid2D_CreateUnitToWorldTransform(
    Camera Facing=true
    Local Pivot:Vector2f=<0.5 0.1>
    Offset:NiagaraPosition=$Engine.Owner.Position
    Rotation:Quat4f=$Engine.Owner.Rotation
    WorldGridExtents:Vector2f=$Emitter.Grid2D_SetResolution.WorldGridExtents
   ) {enabled=false}
  ]
  particleSpawnScript [
   /Niagara/Modules/Spawn/Initialization/InitializeParticle(
    Color=<rgba 1111 13.6009 0 1>
    Lifetime=&RandomRangeFloat(Minimum=0.5, Maximum=1.1)
    Mass=&RandomRangeFloat(Minimum=0.75, Maximum=1)
    Module.Write Color=true
    Module.Write Lifetime=true
    Module.Write Mass=true
    Module.Write Material Random=true
    Module.Write Position=true
    Module.Write Scale=false
    Module.Write SpriteRotation=false
    Module.Write SpriteSize=true
    Sprite Size=&Vector2DFromFloat(Value=1)
   )
   &AddVelocityInCone(
    Cone Angle=80
    Cone Axis=<0 0 1>
    Cone Axis Coordinate Space=::World
    Module.Use Velocity Falloff On Cone Axis=true
    Velocity Distribution Along Cone Axis=0
    Velocity Falloff Away From Cone Axis=0
    Velocity Strength=&Multiply_Float(A=&RandomRangeFloat(Minimum=1075, Maximum=1756), B=$User.Global Scale)
   )
   &SphereLocation(
    MaskSpawnGroup=false
    Override Seed=false
    Sphere Radius=&Multiply_Float(
     A=&Lerp_Float(A=25, B=50, Alpha=&Sine(Normalized Angle=$Emitter.Age, Period=4, Scale=0.5, Bias=1))
     B=$User.Global Scale
    )
    Surface Only=false
    UseNonUniformScale=true
    UseOffset=true
    UseSphereOrientation=false
   ) {enabled=false}
   /Niagara/Modules/Update/Forces/CurlNoiseForce(
    Noise Frequency:NiagaraFloat=&Divide_Float(A=0.04, B=$User.Global Scale)
    Noise Strength=&Multiply_Float(A=250, B=$User.Global Scale)
    Pan Noise=true
    Pan Noise Field=<0.3 0.5 0.3>
   ) {enabled=false}
   &InheritVelocity(Module.Limit Speed=true, Velocity Limit=1500, Velocity Scale=<0 0 0>)
   &ApplyInitialForces()
   &TorusLocation(
    Align Torus With Axis=true
    Large Radius:NiagaraFloat=&Multiply_Float(A=60, B=$User.Global Scale)
    MaskSpawnGroup=false
    Torus Axis=<1 0 0>
    Torus Distribution Mode:ENiagaraTorusDistributionMode=::NewEnumerator1
    Torus Mode=::NewEnumerator2
    U Position:NiagaraFloat=&Subtract_Float(A=&ReturnNormalizedExecIndex(), B=2)
   )
   &AddVelocityFromPoint(
    Module.Use Falloff Distance=false
    Origin Offset=&Multiply_VectorByFloat(Vector=$Engine.Owner.SystemXAxis, Float=&RandomRangeFloat(Minimum=0, Maximum=-60))
    Origin Offset Coordinate Space=::Simulation
    Velocity Origin:Vector3f=&SimulationPosition()
    Velocity Strength=&Multiply_Float(A=&UniformRangedFloat(Minimum=5, Maximum=10), B=$User.Global Scale)
   ) {module=AddVelocityFromPoint001}
  ]
  particleUpdateScript [
   &ParticleState()
   &CurlNoiseForce(Noise Strength=300, Pan Noise=true, Pan Noise Field=<0.1 0.5 0.2>) {module=CurlNoiseForce001}
   &GravityForce(Gravity=&Multiply_VectorByFloat(Vector=<0 0 -1500>, Float=$User.Global Scale))
   &Drag(Drag=&Multiply_Float(A=1.4, B=$User.Global Scale), Use Linear Drag=true, Use Rotational Drag=false)
   &ScaleColor(
    Scale Alpha=&FloatFromCurve(FloatCurve=DataInterface() {propertyValues {Curve {keys [<at -0.0046 1.0113 RCIM_Cubic>,<at 1.0011 0.002 RCIM_Cubic>]}}})
    ScaleA=true
    ScaleRGB=true
   )
   &Collision(
    Age Colliding Particles=false
    Collision Enabled=&SetBoolByFloatComparison(A=$Particles.Age, B=0.4)
    Enable Rest State=false
    Friction=0.2
    Friction During a Bounce=0.3
    Kill Offscreen Particles=true
    Maximum time offscreen=99
    Module.EnableMaxCollisionCount=false
    Module.Randomize Collision Normal Vector=false
    Restitution=0.2
   ) {module=Collision001}
   &SolveForcesAndVelocity(Clamp Velocity=false, Limit Acceleration=false)
   /NiagaraFluids/Modules/Particles/Fluids_Gas_Source(
    NoiseIntensity=3
    NoiseScale=3
    Radius:NiagaraFloat=&FloatFromCurve(
     FloatCurve=DataInterface() {propertyValues {Curve {keys [<at 0 0>, {interpMode=RCIM_Cubic, leaveTangent=-2.282, tangentMode=RCTM_Break, time=0.0361, value=0.984}, <at 1.0022 0.3>]}}}
     Scale Curve=&Multiply_Float(A=&RandomRangeFloat(Minimum=30, Maximum=40), B=$User.Global Scale)
    )
    Temperature:NiagaraFloat=&FloatFromCurve(
     FloatCurve=DataInterface() {propertyValues {Curve {keys [<at 0 1>,<at 0.228 0.772>,<at 1 0>]}}}
     Scale Curve=&RandomRangeFloat(Minimum=500, Maximum=3000)
    )
    VelocityScale:NiagaraFloat=&RandomRangeFloat(Minimum=0.5, Maximum=1)
   )
   /SetVariables(Particles.SpriteSize=&Vector2DFromFloat(Value=&Multiply_Float(A=$Particles.fluids_source_radius, B=&Multiply_Float(A=1.4, B=1.1)))) {enabled=false, isSetParameters=true}
  ]
  rendererClasses [&NiagaraSpriteRendererProperties]
  renderers [{rendererClass=&NiagaraSpriteRendererProperties}]
 }
 Sparkles_and_Black_Chunks {
  bEnabled=false
  simTarget=GPUComputeSim
  emitterUpdateScript [
   &EmitterState(Life Cycle Mode=::NewEnumerator1, Loop Behavior=::NewEnumerator1, Loop Delay=0.0829, Loop Duration=1.5729, UseLoopDelay=true)
   &SpawnBurst_Instantaneous(Spawn Count=300, Use Spawn Probability=false, UseLoopCountLimit=false)
   &SpawnBurst_Instantaneous(Spawn Count=15, Spawn Group=1, Spawn Time=0.04, Use Spawn Probability=false) {module=SpawnBurst_Instantaneous001}
  ]
  particleSpawnScript [
   /SetVariables(Particles.VisibilityTag=$Emitter.SpawnGroup) {isSetParameters=true}
   /Niagara/Modules/Spawn/Initialization/V2/InitializeParticle(
    Color=&MaskLinearColorBySpawnGroup(Masked Linear Color=$User.Global Color, Passthrough Linear Color=<rgba 0 0 0 5>)
    Lifetime=&MaskFloatBySpawnGroup(
     Masked Float=&FloatFromCurve(
      FloatCurve=DataInterface() {propertyValues {Curve {keys [<at 0 0.25 RCIM_Constant>,<at 0.3 0.45>,<at 0.7 1 RCIM_Constant>,<at 0.85 1.5>,<at 1 2>]}}}
      CurveIndex=&RandomRangeFloat(Minimum=0, Maximum=1)
      Scale Curve=0.4
     )
     Passthrough Float=0.05
    )
    Sprite Rotation Mode=::NewEnumerator2
    Sprite Size=&MaskVector2dBySpawnGroup(
     Masked Vector2D=&RandomRangeVector2D(Minimum=<5 5>, Maximum=<6 9>)
     Passthrough Vector2D=&Multiply_Vector2D_ByFloat(Float=&RandomRangeFloat(Minimum=40, Maximum=50), Vector2D=<1 1>)
     Spawn Group Mask=0
    )
    Sprite Size Mode=::NewEnumerator4
    UsePositionOffset=false
    Write Material Random=true
   )
   /Niagara/Modules/Spawn/Location/SphereLocation(
    MaskSpawnGroup=false
    Module.Override Randomness=false
    Sphere Radius=&Multiply_Float(A=0, B=$User.Global Scale)
    Surface Only=false
   )
   &AddVelocity(
    Module.Use Falloff Distance=false
    Override Seed=false
    Velocity Mode=::NewEnumerator1
    Velocity Speed=&Multiply_Float(A=&RandomRangeFloat(Minimum=4000, Maximum=4000), B=$User.Global Scale)
   )
   &ApplyInitialForces(Apply Rotational Force to Rotational Velocity=false)
   /SetVariables(Particles.SpriteSize=&Multiply_Vector2D_ByFloat(Float=$User.Global Scale, Vector2D=$StackContext.SpriteSize)) {isSetParameters=true}
   &CameraOffset(Camera Offset Amount=&Multiply_Float(A=&MaskFloatBySpawnGroup(Masked Float=10, Passthrough Float=100), B=$User.Global Scale))
  ]
  particleUpdateScript [
   &ParticleState()
   &GravityForce(Gravity=&RandomRangeVector(Minimum=<0 0 -900>, Maximum=<0 0 -500>))
   &Drag(
    Drag=&MaskFloatBySpawnGroup(Masked Float=&RandomRangeFloat(Minimum=4, Maximum=9), Passthrough Float=20)
    Use Linear Drag=true
    Use Rotational Drag=false
   )
   &Collision(
    Age Colliding Particles=false
    Kill Offscreen Particles=true
    Module.Control Rest State Via Surface Normal=false
    Module.EnableMaxCollisionCount=false
    Module.Randomize Collision Normal Vector=false
   )
   /Niagara/Modules/Update/Forces/V2/CurlNoiseForce(Noise Frequency=15, Noise Quality / Cost=::NewEnumerator0, Noise Strength=500, Pan Noise=false)
   &SolveForcesAndVelocity(Clamp Velocity=false, Limit Acceleration=false)
   &ScaleColor(
    Scale Alpha=&FloatFromCurve(FloatCurve=DataInterface() {propertyValues {Curve {keys [
         {arriveTangent=26.7273, interpMode=RCIM_Cubic, leaveTangent=26.7273, tangentMode=RCTM_User, time=0.1, value=0.0144}
         {interpMode=RCIM_Cubic, tangentMode=RCTM_User, time=0.15, value=1}
         {interpMode=RCIM_Cubic, tangentMode=RCTM_User, time=0.85, value=1}
         {arriveTangent=-1.3701, interpMode=RCIM_Cubic, leaveTangent=-1.3701, tangentMode=RCTM_User, time=1.0011, value=0}
        ]}}})
    Scale RGB=&Lerp_Vector(
     A=<0 0 0>
     B=<12.5 4 0.8>
     Alpha=&FloatFromCurve(FloatCurve=DataInterface() {propertyValues {Curve {keys [
          {arriveTangent=15.3906, interpMode=RCIM_Cubic, leaveTangent=15.3906, tangentMode=RCTM_User, time=-0.0092, value=1}
          {interpMode=RCIM_Cubic, tangentMode=RCTM_User, time=0.0967, value=0.8}
          {interpMode=RCIM_Cubic, tangentMode=RCTM_User, time=0.45, value=0.8}
          <at 0.7 0 RCIM_Cubic>
         ]}}})
    )
    ScaleA=true
    ScaleRGB=true
   )
   &ScaleSpriteSize(Scale Factor=&Vector2DFromFloat(Value=&FloatFromCurve(FloatCurve=DataInterface() {propertyValues {Curve {keys [<at -0.0046 0>,<at 0.025 1>,<at 0.66 1>,<at 1.0011 0.5>]}}}, Scale Curve=2)))
   &ScaleSpriteSizeBySpeed(
    Max Scale Factor=&MaskVector2dBySpawnGroup(Masked Vector2D=<0.5 2.2>, Passthrough Vector2D=<1 1>, Spawn Group Mask=0)
    Min Scale Factor=&MaskVector2dBySpawnGroup(Masked Vector2D=<1 0.6>, Passthrough Vector2D=<1 1>, Spawn Group Mask=0)
    Velocity Threshold=1500
   )
  ]
  rendererClasses [&NiagaraSpriteRendererProperties]
  renderers [
   {rendererClass=&NiagaraSpriteRendererProperties}
   {rendererClass=&NiagaraSpriteRendererProperties, rendererIndex=1}
  ]
 }
 Streaks_Fat001 {
  bEnabled=false
  emitterUpdateScript [
   &EmitterState(Life Cycle Mode=::NewEnumerator1, Loop Behavior=::NewEnumerator1, Loop Delay=0.1187, Loop Duration=0.4, UseLoopDelay=true)
   &SpawnBurst_Instantaneous(Spawn Count=4, Use Spawn Probability=false, UseLoopCountLimit=false)
   &SpawnBurst_Instantaneous(Spawn Count=2, Spawn Group=1, Use Spawn Probability=false, UseLoopCountLimit=false) {module=SpawnBurst_Instantaneous001}
   &SpawnBurst_Instantaneous(Spawn Count=4, Spawn Group=2, Use Spawn Probability=false) {module=SpawnBurst_Instantaneous002}
  ]
  particleSpawnScript [
   /SetVariables(Particles.VisibilityTag=$Emitter.SpawnGroup) {isSetParameters=true}
   &InitializeParticle(
    Color=<rgba 0.4635 0.4395 0.4204 1.2>
    Lifetime=&MaskFloatBySpawnGroup(
     Masked Float=&RandomRangeFloat(Minimum=0.1, Maximum=0.1)
     Passthrough Float=&RandomRangeFloat(Minimum=0.6, Maximum=1.3)
     Spawn Group Mask=1
    )
    Sprite Size=&Multiply_Vector2D_ByFloat(
     Float=$User.Global Scale
     Vector2D=&MaskVector2dBySpawnGroup(
      Masked Vector2D=&UniformRangedVector2D(Minimum=<100 150>, Maximum=<140 200>)
      Passthrough Vector2D=&UniformRangedVector2D(Minimum=<85 200>, Maximum=<160 300.01>)
      Spawn Group Mask=1
     )
    )
    Sprite Size Mode=::NewEnumerator4
    UsePositionOffset=false
    Write Material Random=true
   )
   &SphereLocation(
    MaskSpawnGroup=false
    Module.Override Randomness=false
    Offset=<0 0 0.01>
    Sphere Radius=&Multiply_Float(A=6, B=$User.Global Scale)
    Surface Only=true
   )
   &AddVelocityFromPoint(
    Module.Use Falloff Distance=false
    Origin Offset=&Multiply_VectorByFloat(Vector=<0 0 -1>, Float=$User.Global Scale)
    Velocity Strength=&Multiply_Float(A=&UniformRangedFloat(Minimum=400, Maximum=500), B=$User.Global Scale)
   )
   &SubUVAnimation(End Frame=3, SubUV Animation Mode:ENiagara_SubUVLookupMode=::NewEnumerator2)
   &CameraOffset(Camera Offset Amount=&Multiply_Float(A=&MaskFloatBySpawnGroup(Masked Float=10.01, Passthrough Float=0, Spawn Group Mask=1), B=$User.Global Scale))
  ]
  particleUpdateScript [
   &ParticleState()
   &Drag(
    Drag=&FloatFromCurve(FloatCurve=DataInterface() {propertyValues {Curve {keys [<at 0.2 1>,<at 0.3 0>]}}}, Scale Curve=15)
    Use Linear Drag=true
    Use Rotational Drag=false
   )
   &GravityForce(Gravity=&Multiply_VectorByFloat(
     Vector=<0 0 -2.2>
     Float=&FloatFromCurve(
      FloatCurve=DataInterface() {propertyValues {Curve {keys [{arriveTangent=3.1357, interpMode=RCIM_Cubic, leaveTangent=3.1357, tangentMode=RCTM_User, time=0.1, value=0}, <at 0.6 1>]}}}
      CurveIndex=$Emitter.Age
     )
    )) {enabled=false}
   &SolveForcesAndVelocity(Clamp Velocity=false, Limit Acceleration=false)
   &ScaleSpriteSize(Scale Factor=&Vector2DFromCurve(
     Vector2Curve=DataInterface() {propertyValues {
       XCurve {keys [
         {arriveTangent=0.4813, interpMode=RCIM_Cubic, leaveTangent=0.4813, tangentMode=RCTM_User, time=0, value=0.625}
         {arriveTangent=1.25, interpMode=RCIM_Cubic, leaveTangent=1.25, time=0.4, value=0.9716}
         {arriveTangent=2.6138, interpMode=RCIM_Cubic, leaveTangent=2.6138, tangentMode=RCTM_User, time=0.9, value=1.75}
        ]}
       YCurve {keys [
         {arriveTangent=9.4573, interpMode=RCIM_Cubic, leaveTangent=9.4573, tangentMode=RCTM_User, time=0, value=0}
         {arriveTangent=0.8193, interpMode=RCIM_Cubic, leaveTangent=0.8193, tangentMode=RCTM_User, time=0.2, value=1.5019}
         <at 1 1.7992 RCIM_Cubic>
        ]}
      }}
     ScaleCurve=<0.8 1.5>
    ))
   &ScaleColor(
    Linear Color Curve=DataInterface() {propertyValues {
      AlphaCurve {keys [
        {arriveTangent=-0.3577, interpMode=RCIM_Cubic, leaveTangent=-0.3577, tangentMode=RCTM_User, time=0.08, value=1}
        {arriveTangent=-1.087, interpMode=RCIM_Cubic, leaveTangent=-1.0164, tangentMode=RCTM_Break, time=0.3702, value=1}
        <at 1 0>
       ]}
      BlueCurve {keys [<at 0.1756 0.4583>,<at 0.3677 0.2>,<at 0.6491 0.3>,<at 1 0.45>]}
      GreenCurve {keys [<at 0.1756 0.4464>,<at 0.3677 0.2>,<at 0.6491 0.3>,<at 1 0.45>]}
      RedCurve {keys [<at 0.1756 0.4471>,<at 0.3677 0.2>,<at 0.6491 0.3>,<at 1 0.45>]}
     }}
    Scale Mode=::NewEnumerator2
   )
   &DynamicMaterialParameters(
    Index 0 Param 1=&Multiply_Float(A=100.01, B=$User.Global Scale)
    Index 0 Param 4=&FloatFromCurve(FloatCurve=DataInterface() {propertyValues {Curve {keys [{interpMode=RCIM_Cubic, tangentMode=RCTM_User, time=0, value=0}, {interpMode=RCIM_Cubic, tangentMode=RCTM_User, time=0.4, value=1}]}}})
    Index 1 Param 1=&FloatFromCurve(
     FloatCurve=DataInterface() {propertyValues {Curve {keys [
         {interpMode=RCIM_Cubic, tangentMode=RCTM_User, time=0, value=0}
         {arriveTangent=0.6295, interpMode=RCIM_Cubic, leaveTangent=0.6295, tangentMode=RCTM_User, time=1, value=1}
        ]}}}
     Scale Curve=0.42
    )
    Index 1 Param 2=&FloatFromCurve(
     FloatCurve=DataInterface() {propertyValues {Curve {keys [
         {arriveTangent=-100.8588, interpMode=RCIM_Cubic, leaveTangent=-100.8588, tangentMode=RCTM_User, time=0, value=30}
         {arriveTangent=-26.7969, interpMode=RCIM_Cubic, leaveTangent=-26.7969, tangentMode=RCTM_User, time=0.15, value=4.3564}
         {arriveTangent=-2.6967, interpMode=RCIM_Cubic, leaveTangent=-2.6967, tangentMode=RCTM_User, time=0.3, value=0}
        ]}}}
     CurveIndex=$Emitter.Age
     Scale Curve=&MaskFloatBySpawnGroup(Masked Float=0, Passthrough Float=5, Spawn Group Mask=1)
    )
    Index 1 Param 3=&FloatFromCurve(
     FloatCurve=DataInterface() {propertyValues {Curve {keys [{interpMode=RCIM_Cubic, tangentMode=RCTM_User, time=0.2, value=0.5}, {interpMode=RCIM_Cubic, tangentMode=RCTM_User, time=1, value=1}]}}}
     Scale Curve=2
    )
    Index 1 Param 4=&FloatFromCurve(FloatCurve=DataInterface() {propertyValues {Curve {keys [<at 0 1>,<at 1 0>]}}}, Scale Curve=-0.3)
    Module.Param0WriteEnabled=true
    Module.Param0WriteEnabled1=true
    Module.Param1WriteEnabled=false
    Module.Param1WriteEnabled1=true
    Module.Param2WriteEnabled=false
    Module.Param2WriteEnabled1=true
    Module.Param3WriteEnabled=true
    Module.Param3WriteEnabled1=true
    Write Parameter Index 1=true
   )
  ]
  rendererClasses [&NiagaraSpriteRendererProperties]
  renderers [
   {rendererClass=&NiagaraSpriteRendererProperties}
   {rendererClass=&NiagaraSpriteRendererProperties, rendererIndex=1}
  ]
 }
}
summary {
 systemName=NS_OLD_Grenade_Explosion_Asteria
 emitters [
  {
   bEnabled=false
   emitterName=Smoke_Trail_Leader
   simTarget=GPUComputeSim
   rendererClasses [&NiagaraSpriteRendererProperties, &NiagaraMeshRendererProperties]
  }
  {
   bEnabled=false
   emitterName=GlowSphereMesh
   rendererClasses [&NiagaraMeshRendererProperties, &NiagaraSpriteRendererProperties]
  }
  {bEnabled=false, emitterName=Intro_Flash, rendererClasses [&NiagaraMeshRendererProperties]}
  {bEnabled=false, emitterName=Flash_wide, rendererClasses [&NiagaraMeshRendererProperties]}
  {bEnabled=false, emitterName=Sparkles001, simTarget=GPUComputeSim, rendererClasses [&NiagaraSpriteRendererProperties]}
  {bEnabled=false, emitterName=Smoke_Ground, rendererClasses [&NiagaraSpriteRendererProperties]}
  {bEnabled=false, emitterName=Light, rendererClasses [&NiagaraLightRendererProperties]}
  {bEnabled=false, emitterName=NE_Explosion_Main, rendererClasses [&NiagaraMeshRendererProperties]}
  {
   bEnabled=false
   emitterName=Sparkles_and_Black_Chunks
   simTarget=GPUComputeSim
   rendererClasses [&NiagaraSpriteRendererProperties]
  }
  {bEnabled=false, emitterName=SourceParticles, simTarget=GPUComputeSim, rendererClasses [&NiagaraSpriteRendererProperties]}
  {
   bEnabled=false
   emitterName=Grid2D_Gas_SmokeFire_Emitter
   simTarget=GPUComputeSim
   rendererClasses [&NiagaraSpriteRendererProperties]
  }
  {
   bEnabled=false
   emitterName=Flash
   rendererClasses [&NiagaraSpriteRendererProperties, &NiagaraMeshRendererProperties]
  }
  {
   bEnabled=false
   emitterName=Smoke_Ground001
   rendererClasses [&NiagaraSpriteRendererProperties, &NiagaraLightRendererProperties]
  }
  {bEnabled=false, emitterName=NE_Sparks_GPU, simTarget=GPUComputeSim, rendererClasses [&NiagaraSpriteRendererProperties]}
  {bEnabled=false, emitterName=NE_Explosion_Main001, rendererClasses [&NiagaraMeshRendererProperties]}
  {bEnabled=false, emitterName=Streaks_Fat001, rendererClasses [&NiagaraSpriteRendererProperties]}
 ]
 userVariables [
  User.Global Color {type=LinearColor, description="", defaultValue {struct=LinearColor, value=<rgba 4 0.5 0 1>}}
  User.Global Scale {type=NiagaraFloat, defaultValue=1, description=""}
  User.PhysicalSurface {type=NiagaraInt32, defaultValue=30}
 ]
}
